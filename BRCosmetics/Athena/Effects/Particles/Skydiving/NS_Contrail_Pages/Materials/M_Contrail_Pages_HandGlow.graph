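# Translucent Niagara sprite material for the hand glow in the Pages skydiving contrail effect.
# A flare streak texture rotates over time at a rate and phase driven by the dynamic parameter Rotation, blending two colors tinted by particle color into emissive.
# Opacity multiplies the streaks by a sharpened vertical UV gradient, then applies depth fade, particle alpha and an Opacity scalar.
# The UV offset uses Offset U on both axes, leaving Offset V unused.
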Material /BRCosmetics/Athena/Effects/Particles/Skydiving/NS_Contrail_Pages/Materials/M_Contrail_Pages_HandGlow
graph {
 properties {bAutomaticallySetUsageInEditor=false, bUsedWithNiagaraSprites=true, blendMode=BLEND_Translucent, usedShadingModels=Default Lit}
 nodes [
  n1 = VectorParameter() {parameterName=Color_2, defaultValue=<rgba 0 0.3141 0.5966 1>}
  n2 = VectorParameter() {parameterName=Color_1, defaultValue=<rgba 0 0.6517 0.8825 1>}
  n3 = VectorParameter() {parameterName=UV Control, channelNames {a="", b="", g=Offset V, r=Offset U}}
  n4 = Add(A=TextureCoordinate(), B=AppendVector(A=@n3.Offset U, B=@n3.Offset U))
  n5 = DynamicParameter() {parameterIndex=1, paramNames [Rotation,Param2,Param3,Param4]}
  n6 = Add(A=Time(), B=LinearInterpolate(Alpha=@n5.Rotation) {constB=5})
  n7 = Multiply(A=@n5.Rotation) {constB=0.08}
  n8 = TextureSample(UVs=MaterialFunctionCall(UVs=@n4, "Rotation Angle (0-1)"=Multiply(A=@n6, B=@n7) {constB=0.05}) {materialFunction=&CustomRotator}.Rotated Values) {texture=/Game/Effects/Fort_Effects/Textures/Flares/T_Flare_Streaks}
  n9 = ParticleColor()
  MaterialFunctionCall(Emissive=Multiply(A=LinearInterpolate(A=@n1.RGB, B=@n2.RGB, Alpha=@n8.R), B=@n9.RGB)) {materialFunction=&MF_FX_EmissiveAdjustments}
  n10 = ComponentMask(None=@n4) {g=true}
  n11 = Multiply(A=Add(A=OneMinus(None=@n10)) {constB=-0.325}) {constB=5}
  n12 = Multiply(A=Saturate(None=Power(Base=@n11) {constExponent=20}), B=@n8.R)
  Multiply(A=@n9.A, B=Multiply(A=DepthFade(Opacity=@n12) {fadeDistanceDefault=10}, B=ScalarParameter() {defaultValue=1, parameterName=Opacity}))
 ]
}
