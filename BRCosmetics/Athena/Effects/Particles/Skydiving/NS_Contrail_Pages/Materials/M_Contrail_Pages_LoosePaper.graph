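# Two-sided lit material for loose paper pages in the Pages skydiving contrail, used on Niagara mesh particles.
# Vertex color red blends two sine-driven hinge rotations, weighted by inverted U, with an added bend and time-based bob, plus rotation-corrected normals.
# Dynamic parameters scale the flap, offset bob timing and blend grey noise-and-gradient paper toward cyan, feeding a dimmed emissive through an adjustment function.
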Material /BRCosmetics/Athena/Effects/Particles/Skydiving/NS_Contrail_Pages/Materials/M_Contrail_Pages_LoosePaper
graph {
 properties {
  bAutomaticallySetUsageInEditor=false
  bUsedWithNiagaraMeshParticles=true
  previewMesh=/BRCosmetics/Athena/Effects/Particles/Skydiving/NS_Contrail_Pages/Meshes/SM_Contrails_Pages_PaperBirds
  twoSided=true
  usedShadingModels=Default Lit
 }
 nodes [
  n1 = WorldPosition()
  CrossProduct(A=DDX(Value=@n1.XYZ), B=DDY(Value=@n1.XYZ))
  n2 = Normalize(VectorInput=Transform(None=Constant3Vector() {constant=<rgba 0 1 0 0>}) {transformSourceType=TRANSFORMSOURCE_Local})
  n3 = MaterialFunctionCall(Speed=Constant() {r=3}) {materialFunction=&DebugTimeSine}
  n4 = LinearInterpolate(
   A=LinearInterpolate(Alpha=@n3.Result) {constA=0.15, constB=-0.05}
   B=LinearInterpolate(Alpha=@n3.Result) {constA=0.85, constB=1.05}
   Alpha=VertexColor().R
  )
  n5 = TextureCoordinate()
  n6 = Saturate(None=OneMinus(None=ComponentMask(None=@n5) {r=true}))
  n7 = Multiply(
   A=RotateAboutAxis(
    NormalizedRotationAxis=@n2
    RotationAngle=@n4
    PivotPoint=MaterialFunctionCall() {materialFunction=&ObjectPivotPoint}.Object Pivot Location
    Position=WorldPosition().XYZ
   )
   B=@n6
  )
  n8 = Constant()
  n9 = Power(Base=@n6) {constExponent=0.8}
  n10 = MaterialFunctionCall(R=@n8, G=@n8, B=LinearInterpolate(Alpha=Saturate(None=@n9)) {constA=-4, constB=0}) {materialFunction=&AppendMany}
  n11 = Constant()
  n12 = Time()
  n13 = DynamicParameter() {defaultValue=<rgba 0 1 0.5 1>, paramNames [GlowTransition, Flapping, RandomFlapping, Param4]}
  n14 = Sine(None=Multiply(A=Add(A=@n12, B=@n13.RandomFlapping) {constB=0.5}) {constB=2.5})
  n15 = LinearInterpolate(Alpha=Divide(A=Add(A=@n14) {constB=0})) {constA=-5, constB=5}
  LinearInterpolate(
   A=Constant()
   B=Add(
    A=Add(A=@n7, B=@n10.RGB)
    B=MaterialFunctionCall(R=@n11, G=@n11, B=@n15) {materialFunction=&AppendMany}.RGB
   )
   Alpha=DynamicParameter() {defaultValue=<rgba 0 1 0.5 1>, paramNames [GlowTransition, Flapping, RandomFlapping, Param4]}.Flapping
  )
  n16 = MaterialFunctionCall() {materialFunction=&RadialGradientExponential}
  n17 = TextureCoordinate()
  n18 = LinearInterpolate(
   A=Constant3Vector() {constant=<rgba 0.3193 0.342 0.3238 1>}
   B=Constant3Vector() {constant=<rgba 0.4735 0.5089 0.491 1>}
   Alpha=Multiply(A=OneMinus(None=@n16.RadialGradientExponential), B=TextureSample(UVs=@n17) {texture=/Game/Global/Textures/Default/Noise/T_godray_Noise}.R)
  )
  n19 = Constant3Vector() {constant=<rgba 0 1 0.8966 1>}
  n20 = DynamicParameter() {defaultValue=<rgba 0 1 0.5 1>, paramNames [GlowTransition, Flapping, RandomFlapping, Param4]}
  MaterialFunctionCall(Emissive=Multiply(A=LinearInterpolate(A=@n18, B=@n19, Alpha=@n20.GlowTransition)) {constB=0.05}) {materialFunction=&MF_FX_EmissiveAdjustments}
  MaterialFunctionCall(Rotation Axis=@n2, Rotation Angle=@n4, Normals Pivot=Constant3Vector()) {materialFunction=&FixRotateAboutAxisNormals}
 ]
}
